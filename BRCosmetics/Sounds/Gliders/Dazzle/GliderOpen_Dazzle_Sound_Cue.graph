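# Sound cue for the opening sound of the Dazzle glider cosmetic in Battle Royale.
# A single wave player plays the GliderOpen_Dazzle_v2 wave at reduced volume, with no randomization or mixing.
# It uses shared glider open/close attenuation and concurrency settings, the cosmetic glider-open property sheet, and the glider music sound class.
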SoundCue /BRCosmetics/Sounds/Gliders/Dazzle/GliderOpen_Dazzle_Sound_Cue
root=WavePlayer_0
nodes [{id=WavePlayer_0, properties {soundWaveAssetPtr=/Game/Sounds/Fort_Gadgets/Athena/Dazzle/GliderOpen_Dazzle_v2}}]
properties {
 attenuationSettings=/Game/Sounds/Attenuation/Fort_Player_Glider_OpenClose_Athena_Att
 audioPropertiesSheet=/BRCosmetics/Sounds/Settings/PropertySheets/Glider/APS_Cosmetic_Glider_Open
 maxDistance=10000
 soundClassObject=/Game/Sounds/FortSoundClassesAndModes/Fort_Gliders_Music
 volumeMultiplier=0.7
 concurrencySet [&Athena_GliderOpenClose_Conc]
}
