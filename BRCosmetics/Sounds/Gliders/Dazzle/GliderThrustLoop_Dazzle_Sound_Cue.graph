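# Looping thrust sound cue for the Dazzle glider cosmetic, with glider attenuation and playback that continues virtualized when silent.
# A mixer layers five looping waves through continuous modulators driven by Forwards, Sideways, Backwards and AnyDirection parameters.
# Forwards drives both a dedicated thrust wave and the shared glider loop; the other parameters each drive that same loop.
# AnyDirection volume mapping is inverted, so its layer fades out as the parameter rises.
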SoundCue /BRCosmetics/Sounds/Gliders/Dazzle/GliderThrustLoop_Dazzle_Sound_Cue
root=Mixer_0
nodes [
 {
  id=Mixer_0
  children [ModulatorContinuous_0, ModulatorContinuous_4, ModulatorContinuous_1, ModulatorContinuous_2, ModulatorContinuous_3]
  properties {inputVolume [1,1,1,1,1]}
 }
 {
  id=ModulatorContinuous_0
  children [WavePlayer_0]
  properties {
   pitchModulationParams {default=0, minOutput=1, parameterName=Forwards}
   volumeModulationParams {default=0, maxOutput=0.8, parameterName=Forwards}
  }
 }
 {
  id=ModulatorContinuous_1
  children [WavePlayer_1]
  properties {
   pitchModulationParams {default=0, minOutput=1, parameterName=Sideways}
   volumeModulationParams {default=0, maxOutput=0.5, parameterName=Sideways}
  }
 }
 {
  id=ModulatorContinuous_2
  children [WavePlayer_2]
  properties {
   pitchModulationParams {default=0, minOutput=1, parameterName=Backwards}
   volumeModulationParams {default=0, maxOutput=0.5, parameterName=Backwards}
  }
 }
 {
  id=ModulatorContinuous_3
  children [WavePlayer_3]
  properties {
   pitchModulationParams {default=0, minOutput=1, parameterName=AnyDirection}
   volumeModulationParams {default=0, maxOutput=0, minOutput=0.5, parameterName=AnyDirection}
  }
 }
 {
  id=ModulatorContinuous_4
  children [WavePlayer_4]
  properties {
   pitchModulationParams {default=0, minOutput=1, parameterName=Forwards}
   volumeModulationParams {default=0, maxOutput=0.9, parameterName=Forwards}
  }
 }
 {id=WavePlayer_0, properties {bLooping=true, soundWaveAssetPtr=/Game/Sounds/Fort_Gadgets/Athena/Dazzle/GliderThrustLoop_Forwards_Dazzle}}
 {id=WavePlayer_1, properties {bLooping=true, soundWaveAssetPtr=/Game/Sounds/Fort_Gadgets/Athena/Dazzle/GliderLoop_Dazzle}}
 {id=WavePlayer_2, properties {bLooping=true, soundWaveAssetPtr=&GliderLoop_Dazzle}}
 {id=WavePlayer_3, properties {bLooping=true, soundWaveAssetPtr=&GliderLoop_Dazzle}}
 {id=WavePlayer_4, properties {bLooping=true, soundWaveAssetPtr=&GliderLoop_Dazzle}}
]
properties {
 attenuationSettings=&Fort_Player_Glider_Thrust_Athena_Att
 audioPropertiesSheet=&APS_Cosmetic_Glider_Loop
 maxDistance=12500
 soundClassObject=/Game/Sounds/FortSoundClassesAndModes/Fort_Gliders_Music
 virtualizationMode=PlayWhenSilent
 volumeMultiplier=0.7
 concurrencySet [null]
}
